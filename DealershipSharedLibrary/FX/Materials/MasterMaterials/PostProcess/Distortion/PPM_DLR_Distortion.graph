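# Post-process material that distorts the screen image with an animated normal map, defaulting to an aquatic water texture.
# Panning normal XY, scaled by a strength parameter, offsets the scene-color lookup UVs, with two blends limiting the offset.
# A square edge mask and a configurable radial vignette provide masks, while MPC_Distortion_PPM drives tiling, pan speed, strength and vignette shape.
# A feature-level switch substitutes a constant on ES2 and ES3.1.
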Material /DealershipSharedLibrary/FX/Materials/MasterMaterials/PostProcess/Distortion/PPM_DLR_Distortion
graph {
 properties {
  bAutomaticallySetUsageInEditor=false
  bUsedWithNiagaraMeshParticles=true
  blendMode=BLEND_Translucent
  materialDomain=MD_PostProcess
  shadingModel=MSM_Unlit
  usedShadingModels=Unlit
 }
 nodes [
  n1 = Constant()
  FeatureLevelSwitch(Default=NamedRerouteUsage(), ES2=@n1, ES3_1=@n1, SM4_REMOVED=@n1)
  n2 = ScreenPosition()
  n3 = ScreenPosition()
  n4 = ScreenPosition()
  n5 = CollectionParameter() {collection=&/MPC_Distortion_PPM, parameterName=DistortionTexUV}
  n6 = MaterialFunctionCall!(Float4=@n5) {materialFunction=/Game/Global/MaterialFunctions/MF_BreakComponents}
  n7 = TextureSampleParameter2D(UVs=Panner(Coordinate=Multiply(A=@n4.ViewportUV, B=@n6.RG) {constB=0.2}, Speed=AppendVector(A=@n6.B, B=@n6.B)) {speedX=0.1, speedY=0.02}) {parameterName=DistortionTexture, samplerType=SAMPLERTYPE_Normal, texture=/DealershipSharedLibrary/FX/Textures/Normal/T_DLR_AquaticWater_N}
  n8 = CollectionParameter() {collection=&MPC_Distortion_PPM, parameterName=DistortionStrength}
  n9 = ComponentMask(None=Multiply(A=@n7.RGB, B=@n8) {constB=0.2}) {g=true, r=true}
  n10 = LinearInterpolate(
   A=@n2.ViewportUV
   B=LinearInterpolate(A=Add(A=@n3.ViewportUV, B=@n9), B=@n2.ViewportUV, Alpha=NamedRerouteUsage())
   Alpha=NamedRerouteUsage()
  )
  NamedRerouteDeclaration(None=ComponentMask(None=SceneTexture(UVs=@n10) {sceneTextureId=PPI_PostProcessInput0}.Color) {b=true, g=true, r=true}) {name=UVDistortion, nodeColor=<rgba 1 0.9375 0 1>}
  n11 = Constant() {r=40}
  n12 = Constant() {r=1}
  NamedRerouteDeclaration(None=Saturate(None=MaterialFunctionCall!(UVs, Mult=@n11, Exp=@n12) {materialFunction=/Game/Effects/Fort_Effects/MaterialFunctions/MF_SquareEdgeMask}.SquareMask)) {name=EdgeMask, nodeColor=<rgba 1 0.0183 0.597 1>}
  n13 = CollectionParameter() {collection=&MPC_Distortion_PPM, parameterName=Vignette - Edge}
  n14 = CollectionParameter() {collection=&MPC_Distortion_PPM, parameterName=Vignette - Center}
  n15 = CollectionParameter() {collection=&MPC_Distortion_PPM, parameterName=Vignette - Radius}
  n16 = CollectionParameter() {
   collection=&MPC_Distortion_PPM
   parameterName=Vignette - Hardness
  }
  n17 = CollectionParameter() {collection=&MPC_Distortion_PPM, parameterName=Vignette - Exp}
  n18 = Power(Base=SphereMask(A=ScreenPosition().ViewportUV, B=Constant() {r=0.5}, Radius=@n15, Hardness=@n16) {attenuationRadius=0.71}, Exp=@n17) {constExponent=1}
  NamedRerouteDeclaration(None=Saturate(None=LinearInterpolate(A=@n13, B=@n14, Alpha=@n18))) {name=Vignette, nodeColor=<rgba 1 0.0183 0.597 1>}
 ]
}
